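# Physics asset for the Male Medium Shirtless Warpaint Green cosmetic body, covering only the dynamic bones of its knotted armband.
# A small box on the knot base roots two capsule chains: bottom_01 and bottom_02 via the knot parent, and top_01 and top_02 directly.
# Constraints give narrow twist throughout, with tighter swing near the knot on the bottom chain and looser swing on the top chain.
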PhysicsAsset /BRCosmetics/Characters/Player/Male/Medium/Bodies/M_MED_Shirtless_Warpaint_Green/Meshes/M_MED_Shirtless_Warpaint_Green_Physics
bodies {
 dyn_armband_bottom_01 [{shapeType=Capsule, length=3, radius=1, center=<2.3579 -0.2806 -0.3258>, rotation=<rot 90 0 0>}]
 dyn_armband_bottom_02 [{shapeType=Capsule, length=2.5, radius=1, center=<1.3681 0.0001 -0.0886>, rotation=<rot 90 0 0>}]
 dyn_armband_knot_base [{shapeType=Box, extentX=0.5, extentY=0.5, extentZ=0.5}]
 dyn_armband_knot_parent [{shapeType=Capsule, length=1, radius=1, rotation=<rot 90 0 0>}]
 dyn_armband_top_01 [{shapeType=Capsule, length=1.5, radius=1, center=<1.3052 0.3037 0>, rotation=<rot 90 0 0>}]
 dyn_armband_top_02 [{shapeType=Capsule, length=1, radius=1, center=<0.5414 -0.0261 -0.1905>, rotation=<rot 90 0 0>}]
}
constraints [
 {bone1Name=dyn_armband_knot_parent, bone2Name=dyn_armband_knot_base, swing1LimitDegrees=10, swing2LimitDegrees=10, twistLimitDegrees=3}
 {bone1Name=dyn_armband_bottom_01, bone2Name=dyn_armband_knot_parent, swing1LimitDegrees=10, swing2LimitDegrees=10, twistLimitDegrees=3}
 {bone1Name=dyn_armband_bottom_02, bone2Name=dyn_armband_bottom_01, swing1LimitDegrees=20, swing2LimitDegrees=20, twistLimitDegrees=3}
 {bone1Name=dyn_armband_top_01, bone2Name=dyn_armband_knot_base, swing1LimitDegrees=30, swing2LimitDegrees=30, twistLimitDegrees=3}
 {bone1Name=dyn_armband_top_02, bone2Name=dyn_armband_top_01, swing1LimitDegrees=30, swing2LimitDegrees=30, twistLimitDegrees=3}
]
